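# Niagara swing trail for the BlondeJaw pickaxe, drawn by a single ribbon emitter that follows the skeletal mesh.
# FN_AnimTrail modules spawn short-lived particles from two bone or socket indices and update them against that mesh.
# Ribbons take the system-level MainColor, fade out over lifetime via an alpha curve, and dim at night through a time-of-day scalar.
# System set-variable modules and user reactivity parameters are present, but their logic is not shown.
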NiagaraSystem /BRCosmetics/Weapons/FORT_Melee/Pickaxe_BlondeJaw/FX/NS_Pickaxe_BlondeJaw_Trail
dependencies {
 usedDataInterfaces [&NiagaraDataInterfaceCurve, &NiagaraDataInterfaceSkeletalMesh]
 usedDynamicInputs [
  &MakeCustomFloatFromBool
  &RemapRange
  &Add_Float
  &Multiply_Float
  &Power
  &ClampFloat
  &Divide_Float
  &Subtract_Float
  &LerpLinearColors
 ]
 usedModules [
  &self:SystemSpawnScript.SystemScriptSource.SystemScriptGraph.NiagaraNodeAssignment_0.SetVariables_#1
  &SystemState
  &self:SystemSpawnScript.SystemScriptSource.SystemScriptGraph.NiagaraNodeAssignment_3.SetVariables_#2
  &self:SystemSpawnScript.SystemScriptSource.SystemScriptGraph.NiagaraNodeAssignment_1.SetVariables_#3
  &self:SystemSpawnScript.SystemScriptSource.SystemScriptGraph.NiagaraNodeAssignment_2.SetVariables_#4
 ]
}
emitters {FN_AnimTrail {
  emitterSpawnScript [&FN_RibbonIDOffsetUtility()]
  emitterUpdateScript [
   &EmitterState()
   &FN_SpawnAnimTrail()
  ]
  particleSpawnScript [
   &FN_InitializeAnimTrailBoneIndecies(First Bone or Socket Index=1, Second Bone or Socket Index=0)
   &FN_InitializeAnimTrail(Color=$System.MainColor, Lifetime=0.2, SkeletalMesh=$System.Skeletal Mesh)
  ]
  particleUpdateScript [
   &SolveForcesAndVelocity(Clamp Velocity=false, Limit Acceleration=false)
   &FN_UpdateAnimTrail(Locked Particle RibbonWidth=1, SkeletalMesh=$System.Skeletal Mesh) {module=FN_UpdateAnimTrail001}
   &ScaleColor(
    Scale Alpha=&FloatFromCurve(FloatCurve=DataInterface() {propertyValues {Curve {keys [{interpMode=RCIM_Cubic, tangentMode=RCTM_User, time=0, value=1}, {interpMode=RCIM_Cubic, tangentMode=RCTM_User, time=0.7, value=0}]}}})
    Scale RGB=&VectorFromFloat(Value=&FNTimeOfDayScalar(Night Scale=0.8, Morning Scale=1, Evening Scale=1))
    ScaleA=true
    ScaleRGB=true
   )
  ]
  rendererClasses [&NiagaraRibbonRendererProperties]
  renderers [{rendererClass=&NiagaraRibbonRendererProperties}]
 }}
summary {
 systemName=NS_Pickaxe_BlondeJaw_Trail
 emitters [{emitterName=FN_AnimTrail, rendererClasses [&NiagaraRibbonRendererProperties]}]
 userVariables [
  User.bIsFrontEnd {type=NiagaraBool}
  User.bIsFrontEndPreview {type=NiagaraBool}
  User.DebugReactivity {type=NiagaraBool}
  User.PhysicalSurface {type=NiagaraInt32}
  User.ReactivityStartTime_0 {type=NiagaraFloat, defaultValue=0, description=""}
  User.ReactivityStartTime_1 {type=NiagaraFloat, defaultValue=0, description=""}
  User.ReactivityStartTime_2 {type=NiagaraFloat, defaultValue=0, description=""}
  User.ReactivityStartTime_3 {type=NiagaraFloat, defaultValue=0, description=""}
  User.ReactivityStartTime_4 {type=NiagaraFloat, defaultValue=0, description=""}
  User.RibbonIDOffset {type=NiagaraInt32}
 ]
}
