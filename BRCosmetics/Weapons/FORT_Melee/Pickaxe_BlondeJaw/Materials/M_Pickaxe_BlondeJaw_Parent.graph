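# Parent material for the BlondeJaw pickaxe, adding a hit-reactive emissive glow between the 2023 character parent PreFX and PostFX functions.
# Five start-time parameters drive ramp-up and decay envelopes, summed with a first-hit boost and squashed by a logistic curve into a reactivity percentage.
# Glow color and brightness come from a fake refraction, using normals blended toward directions from a local center point.
# Static switches allow looped or manual reactivity debugging.
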Material /BRCosmetics/Weapons/FORT_Melee/Pickaxe_BlondeJaw/Materials/M_Pickaxe_BlondeJaw_Parent
graph {
 properties {
  bUseMaterialAttributes=true
  bUsedWithMeshDeformer=true
  bUsedWithMorphTargets=true
  bUsedWithSkeletalMesh=true
  pixelDepthOffsetMode=PDOM_Legacy
  shadingModel=MSM_FromMaterialExpression
  usedShadingModels=Default Lit | Subsurface | Clear Coat | Subsurface Profile | Cloth
 }
 nodes [
  n1 = NamedRerouteUsage()
  n2 = OneMinus(None=@n1)
  Add(A=@n2)
  ConstantBiasScale(None=@n2) {bias=0.5}
  n3 = MaterialFunctionCall() {materialFunction=&MF_CharacterParent_2023_PreFX}
  n4 = GetMaterialAttributes(None=@n3.Result) {attributeGetTypes [#1, #2, #3]}
  n5 = NamedRerouteUsage()
  MaterialFunctionCall(
   Material=SetMaterialAttributes!(MaterialAttributes=@n4.MaterialAttributes, Base Color=@n4.MaterialAttributes, Emissive Color=Add(A=@n4.Emissive Color, B=@n5)) {attributeSetTypes [#1, #2]}
   M=@n3.M
   Pre Skinned Local Position=@n3.M
   Distance Blend=@n3.M
   Fresnel=@n3.M
  ) {materialFunction=&MF_CharacterParent_2023_PostFX}
  Multiply(A=@n1) {constB=0.5}
  n6 = Power(Base=NamedRerouteUsage(), Exp=ScalarParameter() {defaultValue=1, parameterName=Glow_OverallBrightnessRampExp}) {constExponent=0.75}
  n7 = VectorParameter() {group=BlondeJaw, parameterName=Reactive_EmissiveColorLow, defaultValue=<rgba 0 3.6422 5 1>}
  n8 = VectorParameter() {group=BlondeJaw, parameterName=Reactive_EmissiveColor, defaultValue=<rgba 0.1962 0.9502 1 1>}
  n9 = CameraVectorWS()
  n10 = Normalize(VectorInput=Subtract(
    A=LocalPosition() {includedOffsets=ExcludeOffsets, localOrigin=InstancePreSkinning}.XYZ
    B=VectorParameter() {group=BlondeJaw, parameterName=GlowCenterPos_Local, defaultValue=<rgba 0 10 50 1>}.RGB
   ))
  n11 = LinearInterpolate(
   A=Transform(None=@n4.Normal)
   B=Transform(None=VertexInterpolator(VS=@n10).PS) {transformSourceType=TRANSFORMSOURCE_Local}
   Alpha=ScalarParameter() {group=BlondeJaw, parameterName=GlowRefract_Smoothing, sliderMax=1}
  )
  n12 = ScalarParameter() {defaultValue=1, group=BlondeJaw, parameterName=GlowRefract_OriginIOR}
  n13 = ScalarParameter() {defaultValue=1, group=BlondeJaw, parameterName=GlowRefract_TargetIOR}
  n14 = ConstantBiasScale(None=DotProduct(
    A=@n9
    B=MaterialFunctionCall!(Ray Direction=@n9, Surface Normal=@n11, Refractive Index Origin=@n12, Refractive Index Target=@n13) {materialFunction=/Engine/Functions/Engine_MaterialFunctions01/Vectors/Refract}.Vector
   ))
  n15 = SmoothStep(Value=Power(Base=OneMinus(None=@n14))) {constMax=1.01}
  n16 = Multiply(
   A=LinearInterpolate(
    A=VectorParameter() {group=BlondeJaw, parameterName=Reactive_EmissiveColorEarly, defaultValue=<rgba 0 0.1526 1 1>}.RGB
    B=LinearInterpolate(A=@n7.RGB, B=@n8.RGB, Alpha=@n15)
    Alpha=@n6
   )
   B=LinearInterpolate(
    A=ScalarParameter() {defaultValue=0.1, group=BlondeJaw, parameterName=GlowRefract_BrightnessMin}
    B=ScalarParameter() {defaultValue=5, group=BlondeJaw, parameterName=GlowRefract_BrightnessMax}
    Alpha=@n15
   )
  )
  NamedRerouteDeclaration(None=Multiply(
    A=Multiply(A=@n6, B=@n16)
    B=ScalarParameter() {defaultValue=1, group=BlondeJaw, parameterName=EnableReactivity, sliderMax=1, sortPriority=-100}
   )) {name=Emissive, nodeColor=<rgba 1 0.2109 0 1>}
  n17 = Frac(None=Multiply(A=Time()) {constB=0.2})
  n18 = ScalarParameter() {defaultValue=1, group=BlondeJaw, parameterName=Debug_Reactivity_NumHits, sliderMax=5, sortPriority=25}
  n19 = Multiply(A=@n18, B=ScalarParameter() {defaultValue=1.3, group=BlondeJaw, parameterName=ReactivitySingleHitStrength})
  n20 = Multiply(A=Saturate(None=@n18), B=ScalarParameter() {defaultValue=0.4, group=BlondeJaw, parameterName=ReactivityFirstHitStrengthBoost}) {constB=0.4}
  n21 = StaticSwitchParameter(
   True=MaterialFunctionCall!(In=Add(A=@n19, B=@n20)) {materialFunction=/Game/Global/MaterialFunctions/MF_SCurveFunctions}."Logistic Function (-1...1)"
   False=ScalarParameter() {defaultValue=1, group=BlondeJaw, parameterName=Debug_Reactivity_DirectAlpha, sliderMax=1, sortPriority=25}
  ) {group=BlondeJaw, parameterName=DebugReactivity_TestRawHitCount, sortPriority=20}
  n22 = Subtract(A=Time(), B=ScalarParameter() {defaultValue=-100, group=BlondeJaw, parameterName=ReactivityStartTime_0})
  n23 = Saturate(None=Divide(A=@n22, B=ScalarParameter() {defaultValue=0.3, group=BlondeJaw, parameterName=ReactivityRampUpDuration}) {constB=0.8})
  n24 = ScalarParameter() {defaultValue=0.8, group=BlondeJaw, parameterName=ReactivityRampUpExp}
  n25 = ScalarParameter() {defaultValue=10, group=BlondeJaw, parameterName=ReactivityRampDownDuration}
  n26 = OneMinus(None=Saturate(None=Divide(A=@n22, B=@n25) {constB=0.8}))
  n27 = ScalarParameter() {defaultValue=2, group=BlondeJaw, parameterName=ReactivityRampDownExp}
  n28 = Subtract(A=Time(), B=ScalarParameter() {defaultValue=-100, group=BlondeJaw, parameterName=ReactivityStartTime_1})
  n29 = ScalarParameter() {defaultValue=0.3, group=BlondeJaw, parameterName=ReactivityRampUpDuration}
  n30 = Power(
   Base=Saturate(None=Divide(A=@n28, B=@n29) {constB=0.8})
   Exp=ScalarParameter() {defaultValue=0.8, group=BlondeJaw, parameterName=ReactivityRampUpExp}
  ) {constExponent=0.8}
  n31 = Divide(A=@n28, B=ScalarParameter() {defaultValue=10, group=BlondeJaw, parameterName=ReactivityRampDownDuration}) {constB=0.8}
  n32 = Power(Base=OneMinus(None=Saturate(None=@n31)), Exp=ScalarParameter() {defaultValue=2, group=BlondeJaw, parameterName=ReactivityRampDownExp}) {constExponent=0.4}
  n33 = Subtract(A=Time(), B=ScalarParameter() {defaultValue=-100, group=BlondeJaw, parameterName=ReactivityStartTime_2})
  n34 = Divide(A=@n33, B=ScalarParameter() {defaultValue=0.3, group=BlondeJaw, parameterName=ReactivityRampUpDuration}) {constB=0.8}
  n35 = Saturate(None=Divide(A=@n33, B=ScalarParameter() {defaultValue=10, group=BlondeJaw, parameterName=ReactivityRampDownDuration}) {constB=0.8})
  n36 = Multiply(
   A=Power(Base=Saturate(None=@n34), Exp=ScalarParameter() {defaultValue=0.8, group=BlondeJaw, parameterName=ReactivityRampUpExp}) {constExponent=0.8}
   B=Power(Base=OneMinus(None=@n35), Exp=ScalarParameter() {defaultValue=2, group=BlondeJaw, parameterName=ReactivityRampDownExp}) {constExponent=0.4}
  )
  n37 = Subtract(A=Time(), B=ScalarParameter() {defaultValue=-100, group=BlondeJaw, parameterName=ReactivityStartTime_3})
  n38 = Saturate(None=Divide(A=@n37, B=ScalarParameter() {defaultValue=0.3, group=BlondeJaw, parameterName=ReactivityRampUpDuration}) {constB=0.8})
  n39 = ScalarParameter() {defaultValue=0.8, group=BlondeJaw, parameterName=ReactivityRampUpExp}
  n40 = ScalarParameter() {defaultValue=10, group=BlondeJaw, parameterName=ReactivityRampDownDuration}
  n41 = OneMinus(None=Saturate(None=Divide(A=@n37, B=@n40) {constB=0.8}))
  n42 = ScalarParameter() {defaultValue=2, group=BlondeJaw, parameterName=ReactivityRampDownExp}
  n43 = Subtract(A=Time(), B=ScalarParameter() {defaultValue=-100, group=BlondeJaw, parameterName=ReactivityStartTime_4})
  n44 = Saturate(None=Divide(A=@n43, B=ScalarParameter() {defaultValue=0.3, group=BlondeJaw, parameterName=ReactivityRampUpDuration}) {constB=0.8})
  n45 = ScalarParameter() {defaultValue=0.8, group=BlondeJaw, parameterName=ReactivityRampUpExp}
  n46 = ScalarParameter() {defaultValue=10, group=BlondeJaw, parameterName=ReactivityRampDownDuration}
  n47 = OneMinus(None=Saturate(None=Divide(A=@n43, B=@n46) {constB=0.8}))
  n48 = ScalarParameter() {defaultValue=2, group=BlondeJaw, parameterName=ReactivityRampDownExp}
  n49 = Add(
   A=Multiply(A=Power(Base=@n38, Exp=@n39) {constExponent=0.8}, B=Power(Base=@n41, Exp=@n42) {constExponent=0.4})
   B=Multiply(A=Power(Base=@n44, Exp=@n45) {constExponent=0.8}, B=Power(Base=@n47, Exp=@n48) {constExponent=0.4})
  )
  n50 = Add(
   A=Multiply(A=Power(Base=@n23, Exp=@n24) {constExponent=0.8}, B=Power(Base=@n26, Exp=@n27) {constExponent=0.4})
   B=Add(A=Multiply(A=@n30, B=@n32), B=Add(A=@n36, B=@n49))
  )
  n51 = Add(
   A=Multiply(A=@n50, B=ScalarParameter() {defaultValue=1.3, group=BlondeJaw, parameterName=ReactivitySingleHitStrength})
   B=Multiply(A=Saturate(None=@n50), B=ScalarParameter() {defaultValue=0.4, group=BlondeJaw, parameterName=ReactivityFirstHitStrengthBoost}) {constB=0.4}
  )
  NamedRerouteDeclaration(None=StaticSwitchParameter(
    True=StaticSwitchParameter(True=@n17, False=@n21) {group=BlondeJaw, parameterName=DebugReactivity_Loop, sortPriority=20}
    False=MaterialFunctionCall(In=@n51) {materialFunction=&MF_SCurveFunctions}."Logistic Function (-1...1)"
   ) {group=BlondeJaw, parameterName=DebugReactivity, sortPriority=20}) {name=RawReactivityPercentage, nodeColor=<rgba 0.8516 1 0 1>}
 ]
}
