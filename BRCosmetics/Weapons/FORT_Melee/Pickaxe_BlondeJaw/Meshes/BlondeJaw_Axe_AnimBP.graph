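# Animation Blueprint for the BlondeJaw pickaxe that adds secondary physics motion to its four-bone dyn_chain.
# The cached input pose feeds two branches, each rotating the root chain bone before running chained Anim Dynamics nodes.
# bIsInFrontEnd blends between a half-strength, stiff-spring, world-space setup for the front end and a full-strength setup otherwise.
# The event graph updates gravity override each frame through a deprecated state function, feeding every dynamics node.
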AnimBlueprint /BRCosmetics/Weapons/FORT_Melee/Pickaxe_BlondeJaw/Meshes/BlondeJaw_Axe_AnimBP
target_skeleton=&/BlondeJaw_Axe_Skeleton
graphs [
 AnimGraph {nodes [
   n1 = "Transform (Modify) Bone - Bone: dyn_chain_01"(
    ComponentPose=Local To Component(LocalPose=Use cached pose 'INPUT'())
    Alpha=1
    Translation="0,0,0"
    Rotation="17.231,-12.587,-18.442"
    Scale="1,1,1"
   ) {
    rotationMode=BMM_Additive
    rotationSpace=BCS_ParentBoneSpace
    alphaBoolBlend {blendOption=Linear}
    boneToModify {boneName=dyn_chain_01}
    rotation=<rot 17.231 -12.587 -18.442>
   }
   n2 = Get Gravity Override()
   n3 = Anim Dynamics - Bone: dyn_chain_01!(ComponentPose=@n1, Alpha=0.5, GravityOverride:Vector=@n2, ExternalForce="0,0,0") -> (Pose) {
    angularBiasOverride=1
    angularDampingOverride=1
    angularSpringConstant=1500
    bAngularSpring=true
    bOverrideAngularBias=true
    bOverrideAngularDamping=true
    bUseGravityOverride=true
    bUsePlanarLimit=false
    gravityScale=0.35
    simulationSpace=World
    alphaBoolBlend {blendOption=Linear}
    boundBone {boneName=dyn_chain_01}
    componentAppliedLinearAccClamp=<1000 1000 1000>
    componentLinearAccScale=<0.75 0.75 0.75>
    componentLinearVelScale=<0.95 0.95 0.95>
    gravityOverride {z=-980}
    physicsBodyDefinitions [{
      collisionType=OuterSphere
      sphereCollisionRadius=1
      boundBone {boneName=dyn_chain_01}
      boxExtents=<12 8 8>
      constraintSetup {angularLimitsMax=<0 6 0>, angularLimitsMin=<0 -6 -15>, angularTarget=<1 0 0>}
      localJointOffset=<1 0 0>
     }]
   }
   n4 = Anim Dynamics - Bone: dyn_chain_02!(ComponentPose=@n3, Alpha=0.5, GravityOverride:Vector=@n2, ExternalForce="0,0,0") -> (Pose) {
    angularBiasOverride=1
    angularDampingOverride=1
    angularSpringConstant=1500
    bAngularSpring=true
    bOverrideAngularBias=true
    bOverrideAngularDamping=true
    bUseGravityOverride=true
    bUsePlanarLimit=false
    gravityScale=0.35
    simulationSpace=World
    alphaBoolBlend {blendOption=Linear}
    boundBone {boneName=dyn_chain_02}
    componentAppliedLinearAccClamp=<1000 1000 1000>
    componentLinearAccScale=<0.75 0.75 0.75>
    componentLinearVelScale=<0.95 0.95 0.95>
    gravityOverride {z=-980}
    physicsBodyDefinitions [{
      collisionType=OuterSphere
      sphereCollisionRadius=1
      boundBone {boneName=dyn_chain_02}
      boxExtents=<12 8 8>
      constraintSetup {angularLimitsMax=<3 6 0>, angularLimitsMin=<-3 -6 -15>, angularTarget=<1 0 0>}
      localJointOffset=<1 0 0>
     }]
   }
   n5 = Anim Dynamics - Bone: dyn_chain_03!(ComponentPose=@n4, Alpha=0.5, GravityOverride:Vector=@n2, ExternalForce="0,0,0") -> (Pose) {
    angularBiasOverride=1
    angularDampingOverride=1
    angularSpringConstant=1500
    bAngularSpring=true
    bOverrideAngularBias=true
    bOverrideAngularDamping=true
    bUseGravityOverride=true
    bUsePlanarLimit=false
    gravityScale=0.35
    simulationSpace=World
    alphaBoolBlend {blendOption=Linear}
    boundBone {boneName=dyn_chain_03}
    componentAppliedLinearAccClamp=<1000 1000 1000>
    componentLinearAccScale=<0.75 0.75 0.75>
    componentLinearVelScale=<0.95 0.95 0.95>
    gravityOverride {z=-980}
    physicsBodyDefinitions [{
      collisionType=OuterSphere
      sphereCollisionRadius=1
      boundBone {boneName=dyn_chain_03}
      boxExtents=<12 8 8>
      constraintSetup {angularLimitsMax=<6 6 0>, angularLimitsMin=<-6 -6 -10>, angularTarget=<1 0 0>}
      localJointOffset=<1 0 0>
     }]
   }
   n6 = Anim Dynamics - Bone: dyn_chain_04!(ComponentPose=@n5, Alpha=0.5, GravityOverride:Vector=@n2, ExternalForce="0,0,0") -> (Pose) {
    angularBiasOverride=1
    angularDampingOverride=1
    angularSpringConstant=1500
    bAngularSpring=true
    bOverrideAngularBias=true
    bOverrideAngularDamping=true
    bUseGravityOverride=true
    bUsePlanarLimit=false
    gravityScale=0.35
    simulationSpace=World
    alphaBoolBlend {blendOption=Linear}
    boundBone {boneName=dyn_chain_04}
    componentAppliedLinearAccClamp=<1000 1000 1000>
    componentLinearAccScale=<0.75 0.75 0.75>
    componentLinearVelScale=<0.95 0.95 0.95>
    gravityOverride {z=-980}
    physicsBodyDefinitions [{
      collisionType=OuterSphere
      sphereCollisionRadius=1
      boundBone {boneName=dyn_chain_04}
      boxExtents=<12 8 8>
      constraintSetup {angularLimitsMax=<6 6 0>, angularLimitsMin=<-6 -6 -10>, angularTarget=<1 0 0>}
      localJointOffset=<1 0 0>
     }]
   }
   n7 = "Transform (Modify) Bone - Bone: dyn_chain_01"(ComponentPose=Local To Component(LocalPose=Use cached pose 'INPUT'()), Alpha=1, Translation="0,0,0", Rotation="-40,25,-25", Scale="1,1,1") {
    rotationMode=BMM_Additive
    rotationSpace=BCS_ParentBoneSpace
    alphaBoolBlend {blendOption=Linear}
    boneToModify {boneName=dyn_chain_01}
    rotation=<rot -40 25 -25>
   }
   n8 = Get Gravity Override()
   n9 = Anim Dynamics - Bone: dyn_chain_01(ComponentPose=@n7, Alpha=1, GravityOverride=@n8, ExternalForce="0,0,0") {
    angularBiasOverride=0.999
    angularDampingOverride=1
    angularSpringConstant=175
    bOverrideAngularDamping=true
    bUseGravityOverride=true
    bUsePlanarLimit=false
    gravityScale=0.35
    alphaBoolBlend {blendOption=Linear}
    boundBone {boneName=dyn_chain_01}
    componentAppliedLinearAccClamp=<1500 1500 1500>
    componentLinearAccScale=<0.65 0.65 0.65>
    componentLinearVelScale=<0.8 0.8 0.8>
    gravityOverride {z=-980}
    physicsBodyDefinitions [{
      collisionType=OuterSphere
      sphereCollisionRadius=1
      boundBone {boneName=dyn_chain_01}
      boxExtents=<12 6 6>
      constraintSetup {angularLimitsMax=<0 3 4>, angularLimitsMin=<0 -6 -10>, angularTarget=<1 0 0>}
      localJointOffset=<1 0 0>
     }]
   }
   n10 = Anim Dynamics - Bone: dyn_chain_02(ComponentPose=@n9, Alpha=1, GravityOverride=@n8, ExternalForce="0,0,0") {
    angularBiasOverride=0.999
    angularDampingOverride=1
    angularSpringConstant=175
    bOverrideAngularDamping=true
    bUseGravityOverride=true
    bUsePlanarLimit=false
    gravityScale=0.35
    alphaBoolBlend {blendOption=Linear}
    boundBone {boneName=dyn_chain_02}
    componentAppliedLinearAccClamp=<1500 1500 1500>
    componentLinearAccScale=<0.65 0.65 0.65>
    componentLinearVelScale=<0.8 0.8 0.8>
    gravityOverride {z=-980}
    physicsBodyDefinitions [{
      collisionType=OuterSphere
      sphereCollisionRadius=1
      boundBone {boneName=dyn_chain_02}
      boxExtents=<12 8 8>
      constraintSetup {angularLimitsMax=<3 3 8>, angularLimitsMin=<-3 -6 -10>, angularTarget=<1 0 0>}
      localJointOffset=<1 0 0>
     }]
   }
   n11 = Anim Dynamics - Bone: dyn_chain_03(ComponentPose=@n10, Alpha=1, GravityOverride=@n8, ExternalForce="0,0,0") {
    angularBiasOverride=0.999
    angularDampingOverride=1
    angularSpringConstant=175
    bOverrideAngularDamping=true
    bUseGravityOverride=true
    bUsePlanarLimit=false
    gravityScale=0.35
    alphaBoolBlend {blendOption=Linear}
    boundBone {boneName=dyn_chain_03}
    componentAppliedLinearAccClamp=<1500 1500 1500>
    componentLinearAccScale=<0.65 0.65 0.65>
    componentLinearVelScale=<0.8 0.8 0.8>
    gravityOverride {z=-980}
    physicsBodyDefinitions [{
      collisionType=OuterSphere
      sphereCollisionRadius=1
      boundBone {boneName=dyn_chain_03}
      boxExtents=<12 8 8>
      constraintSetup {angularLimitsMax=<6 3 14>, angularLimitsMin=<-6 -6 -10>, angularTarget=<1 0 0>}
      localJointOffset=<1 0 0>
     }]
   }
   n12 = Anim Dynamics - Bone: dyn_chain_04(ComponentPose=@n11, Alpha=1, GravityOverride=@n8, ExternalForce="0,0,0") {
    angularBiasOverride=0.999
    angularDampingOverride=1
    angularSpringConstant=175
    bOverrideAngularBias=true
    bOverrideAngularDamping=true
    bUseGravityOverride=true
    bUsePlanarLimit=false
    gravityScale=0.35
    alphaBoolBlend {blendOption=Linear}
    boundBone {boneName=dyn_chain_04}
    componentAppliedLinearAccClamp=<1500 1500 1500>
    componentLinearAccScale=<0.65 0.65 0.65>
    componentLinearVelScale=<0.8 0.8 0.8>
    gravityOverride {z=-980}
    physicsBodyDefinitions [{
      collisionType=OuterSphere
      sphereCollisionRadius=1
      boundBone {boneName=dyn_chain_04}
      boxExtents=<11 7 7>
      constraintSetup {angularLimitsMax=<6 3 16>, angularLimitsMin=<-6 -6 -15>, angularTarget=<1 0 0>}
      localJointOffset=<1 0 0>
     }]
   }
   Output Pose(Result=Blend Poses by bool(
     BlendPose_0=Component To Local(ComponentPose=@n6)
     BlendPose_1=Component To Local(ComponentPose=@n12)
     BlendTime_0=0.3
     BlendTime_1=0.3
     bActiveValue=Get bIsInFrontEnd()
    ) {blendType=Linear}) {layerGroup=None}
   AnimGraphNode_SubInput_0 = Input Pose()
   Save cached pose 'INPUT'(Pose=@AnimGraphNode_SubInput_0) {cacheName=INPUT}
  ]}
 EventGraph {nodes [
   Event BlueprintUpdateAnimation()
   > n1 = Gravity Override State Function DEPRECATED(CharacterPartAnimInstance=Self-Reference(), GameTimeInSeconds=GetGameTimeInSeconds(), GravityOverrideParams=Get Gravity Override Params())
   > Set Gravity Override(Gravity Override=@n1)
   TryGetPawnOwner()
  ]}
]
variables [
 Gravity Override {type=Vector}
 Gravity Override Params {type=GravityOverrideParamsStruct}
]
