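# Physics asset for the BlondeJaw pickaxe mesh, a two-handed melee cosmetic weapon.
# It defines a single capsule body on the Melee_TwoHanded_Handle bone, offset and rotated along the handle's length as a rough collision volume.
# It has no constraints, so there are no jointed or separately simulated parts.
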PhysicsAsset /BRCosmetics/Weapons/FORT_Melee/Pickaxe_BlondeJaw/Meshes/BlondeJaw_Axe_PhysicsAsset
bodies {Melee_TwoHanded_Handle [{shapeType=Capsule, length=65.2535, radius=42.4652, center=<-0.0482 1.7417 31.6226>, rotation=<rot -1.1997 -90 0.7055>}]}
constraints []
